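; Rift spawner actor that brings in a Hoagie vehicle with cosmetic rift effects, then expires.
; After a delay, the server knocks back pawns and vehicles, destroys buildings, async-loads and spawns the fueled vehicle.
; It registers the vehicle with the game state component and marks it for untagged nearby walking players.
; Clients play opening effects on BeginPlay, closing ones on bRiftFXClose replication, with sounds gated by player tags.
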(event Custom|Play%20Rift%20Start%20Sound
  (params (Location :type Vector))
  ;[0 0.03646 0.006314 0.5] BeginPlay: On the server, spawn the vehicle after a cosmetic delay.
  ; On the client, start the cosmetic rift in visuals.
  (bind _returnvalue (Audio|SpawnSoundatLocation (Variables|FX|GetRiftStartSound) Location))
  (Variables|Transient|SetRiftStartSoundInstance _returnvalue)
  (SoundIndicator|StartSoundIndicatorLooping (Variables|Default|GetFortSoundIndicator) Location (Variables|FX|GetRiftStartSound) 1.0 0))

(event EventBeginPlay
  (bind _returnvalue (Utilities|GetVehicleSpawnLocation))
  ;[0 0.03646 0.006314 0.5] BeginPlay: On the server, spawn the vehicle after a cosmetic delay.
  ; On the client, start the cosmetic rift in visuals.
  (|CallParentFunction)
  (Networking|SwitchHasAuthority
    (:Authority
      ;;[0.05158 0.01685 0.15 0.5] On the server, spawn vehicle after a delay and destroy self.
      (Utilities|FlowControl|Delay (Variables|Settings|GetDelayBeforeSpawn))
      (CallFunction|SpawnVehicle)
      (Variables|Transient|SetbRiftFXClose true)
      (Networking|FlushNetDormancy)
      (Actor|SetLifeSpan 1e+01))
    (:Remote
      ;;[0.15 0.01509 0.1234 0.5] On the client, play the rift in visuals.
      (if (not (Variables|Transient|GetbRiftFXClose))
        (bind _returnvalue_2 (Niagara|SpawnSystematLocation :SystemTemplate (Variables|FX|GetRiftOpeningNiagaraSystem) :Location _returnvalue :Scale "0.500000,0.500000,0.500000" :bAutoDestroy true :bAutoActivate true :PoolingMethod "AutoRelease" :bPreCullCheck true))
        (Variables|Transient|SetRiftOpeningFXInstance _returnvalue_2)
        (if (CallFunction|ShouldPlayRiftSounds)
          (CallFunction|PlayRiftStartSound _returnvalue)
          (else
            _))
        (CameraShake|Start (Variables|Default|GetCameraShakeSource))
        (Components|Activation|Activate (Variables|Default|GetForceFeedback) false)))))

(event Custom|SpawnVehicle
  (bind _returnvalue (Fortnite|Athena|GameState|GetGameStateAthena))
  (bind _returnvalue_2 (Actor|GetComponentbyClass _returnvalue "/ToyFoxGameplay/Gameplay/B_ToyFoxGameStateComponent.B_ToyFoxGameStateComponent_C"))
  ;[0.000521 0.15 0.1306 0.5] SpawnVehicle: On the server, spawn the vehicle actor at the desired location.
  (Networking|SwitchHasAuthority
    (:Authority
      (CallFunction|RiftExplosion)
      (Utilities|FlowControl|Delay 0.2) ; Slight delay for rift explosion to play out
      (bind _class (Utilities|AsyncLoadClassAsset (Variables|Settings|GetVehicleClassSoft))
        (:Completed
          (bind _returnvalue_3 (Game|SpawnActorfromClass (Utilities|Casting|CastToActorClass _class) (Math|Transform|MakeTransform (Utilities|GetVehicleSpawnLocation) (Math|Rotator|MakeRotator 0.0 0.0 (Math|Random|RandomFloatinRange 0.0 3.6e+02))) "AdjustIfPossibleButAlwaysSpawn" "MultiplyWithRoot" 0 0))
          (Vehicle|WakeUp(Message) _returnvalue_3)
          (Vehicle|Fuel|SetFuel(Message) _returnvalue_3 1e+02)
          (CallFunction|MarkVehicleforNearbyPlayers (Utilities|Casting|CastToFortAthenaVehicle _returnvalue_3))
          (Utilities|IsValid _returnvalue
            (:Is%20Valid
              (Utilities|IsValid _returnvalue_2
                (:Is%20Valid
                  (Hoagie|RegisterHoagieVehicle _returnvalue_2 (Utilities|Casting|CastToFortAthenaVehicle _returnvalue_3)))))))))))

(event Custom|OnRep_bRiftFXClose
  (bind _returnvalue (Utilities|GetVehicleSpawnLocation))
  (bind _riftclosesound (Variables|FX|GetRiftCloseSound))
  ;[0.15 0 0.1127 0.5] OnRep_bRiftFXClose: Play the rift out visuals.
  (Networking|CanExecuteCosmeticEvents
    (:True
      (if (Variables|Transient|GetbRiftFXClose)
        (CameraShake|StopAllCameraShakes (Variables|Default|GetCameraShakeSource) true)
        (Components|Activation|Deactivate (Variables|Default|GetForceFeedback))
        (Audio|Components|Audio|StopDelayed (Variables|Transient|GetRiftStartSoundInstance) 0.1)
        (Components|Activation|Deactivate (Variables|Transient|GetRiftOpeningFXInstance))
        (Niagara|SpawnSystematLocation :SystemTemplate (Variables|FX|GetRiftExplodeNiagaraSystem) :Location _returnvalue :Scale "1.000000,1.000000,1.000000" :bAutoDestroy true :bAutoActivate true :PoolingMethod "AutoRelease" :bPreCullCheck true)
        (if (CallFunction|ShouldPlayRiftSounds)
          (Audio|PlaySoundatLocation _riftclosesound _returnvalue)
          (else
            _))
        (Utilities|FlowControl|Delay (Variables|/Script/Engine.SoundBase|GetDuration _riftclosesound))
        (SoundIndicator|StopSoundIndicatorLooping (Variables|Default|GetFortSoundIndicator))))))

(fn MarkVehicleForNearbyPlayers
  (params (VehicleActor :type Fort%20Athena%20Vehicle%20Object%20Reference))
  (bind _hotfix_markvehicleforneabyplayersdist (Variables|Hotfix|GetHotfix_MarkVehicleForNeabyPlayersDist))
  (bind _returnvalue (ScalableFloat|GetCachedScalableFloatValue _hotfix_markvehicleforneabyplayersdist))
  ;[0.15 0 0.1272 0.5] MarkVehicleForNearbyPlayers: Marks the vehicle for nearby players that are not already in a vehicle, if desired.
  (Networking|SwitchHasAuthority
    (:Authority
      (Utilities|IsValid VehicleActor
        (:Is%20Valid
          (if (> _returnvalue 0)
            (bind _outfortplayerpawns (Fortnite|GetFortPlayerPawnsInRange (Transformation|GetActorLocation VehicleActor) _returnvalue true))
            (for _array_element _outfortplayerpawns
              (Utilities|IsValid _array_element
                (:Is%20Valid
                  (if (Math|Boolean|ANDBoolean (not (Vehicle|IsinVehicle _array_element)) (Pawn|IsPlayerControlled _array_element) (Pawn|Components|CharacterMovement|IsWalking (Variables|/Script/Engine.Character|GetCharacterMovement _array_element)) (not (Actor|ActorHasTag _array_element (Variables|Settings|GetHoagieMarkedTrackerActorTag))))
                    (Utilities|IsValid (Fortnite|GetFortPlayerControllerfromActor _array_element)
                      (:Is%20Valid
                        (Utilities|Array|Add (Variables|/Script/Engine.Actor|GetTags _array_element) (Variables|Settings|GetHoagieMarkedTrackerActorTag))
                        (bind _returnvalue_2 (Actor|GetComponentbyClass (Fortnite|GetFortPlayerControllerfromActor _array_element) "/Script/FortniteGame.AthenaMarkerComponent"))
                        (Utilities|IsValid _returnvalue_2
                          (:Is%20Valid
                            (MarkerComponent|MarkActoronServer _returnvalue_2 VehicleActor false true)))))))))))))))

(fn RiftExplosion
  ;[0.06013 0.0265 0.15 0.5] RiftExplosion: Knockback pawns and destroy building actors in the way.
  (Networking|SwitchHasAuthority
    (:Authority
      (bind _outactors (Collision|SphereOverlapActors (Utilities|GetVehicleSpawnLocation) (ScalableFloat|GetValueatLevel (Variables|Hotfix|GetHotfix_RiftExplosionRadius)) (Utilities|Array|MakeArray "ObjectTypeQuery1" "ObjectTypeQuery2" "ObjectTypeQuery3" "ObjectTypeQuery5" "ObjectTypeQuery7")))
      (for _array_element _outactors
        (bind _asbuilding_actor (Utilities|Casting|CastToBuildingActor _array_element)
          (:then
            (if (and (Variables|/Script/Engine.Actor|GetbCanBeDamaged _asbuilding_actor) (> (ScalableFloat|GetCachedScalableFloatValue (Variables|Hotfix|GetHotfix_RiftExplosionDestroyBuildings)) 0))
              (Gameplay|Die :self _asbuilding_actor :!class /Script/FortniteGame.BuildingActor)))
          (:CastFailed
            (if (and (or (.bSuccess (Utilities|Casting|CastToFortPawn _array_element)) (.bSuccess (Utilities|Casting|CastToFortAthenaVehicle _array_element))) (> (ScalableFloat|GetCachedScalableFloatValue (Variables|Hotfix|GetHotfix_RiftExplosionKnockback)) 0))
              (Fortnite|ApplyGameplayEffectsToActor _array_element (Utilities|Array|MakeArray (Variables|Settings|GetRiftExplosionKnockbackGameplayEffectClass)) (Utilities|GetVehicleSpawnLocation) false 0 (Ability|GetFortAbilitySystemComponent self) (GameplayEffects|MakeFortEffectContext (Ability|GetFortAbilitySystemComponent self) self self 0 0)))))))))

(fn ShouldPlayRiftSounds
  (bind _returnvalue (Fortnite|GetFirstLocalFortPlayerController true))
  (bind _returnvalue_2 (Game|Player|GetPlayerPawn _returnvalue))
  (Utilities|IsValid _returnvalue
    (:Is%20Valid
      (Utilities|IsValid _returnvalue_2
        (:Is%20Valid
          (return (not (GameplayTags|HasAnyMatchingGameplayTags _returnvalue_2 (Variables|FX|GetRift%20Sound%20Blocking%20Tags)))))
        (:Is%20Not%20Valid
          (return true))))
    (:Is%20Not%20Valid
      (return true))))

(fn ConstructionScript)
